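# Physics asset for the Snowboarder backpack cosmetic, defining collision and joint limits for its dangling accessories.
# The pack body uses two boxes, while the grenade, gun, pickaxe and two keychain bones each use a capsule.
# Grenade, gun and pickaxe attach to the pack with tight swing and twist limits for slight wobble.
# The keychain forms a two-link chain from the pack with wide swing and narrow twist, letting it swing freely.
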PhysicsAsset /BRCosmetics/Accessories/FORT_Backpacks/Backpack_M_MED_Snowboarder/Meshes/M_MED_Snowboarder_Pack_Physics
bodies {
 dyn_gernade [{shapeType=Capsule, length=5.086, radius=1.642, center=<-0.0001 0 -1.693>, rotation=<rot 0 -90 0>}]
 dyn_gun [{shapeType=Capsule, length=5.086, radius=1.642, center=<-0.0001 0 -1.7068>, rotation=<rot 0 -90 0>}]
 dyn_keychain_01 [{shapeType=Capsule, length=0.505, radius=0.505, center=<0.8045 0 0>, rotation=<rot 90 0 0>}]
 dyn_keychain_02 [{shapeType=Capsule, length=4.6533, radius=0.8537, center=<3.1009 0 0>, rotation=<rot 90 0 0>}]
 dyn_pickaxe [{shapeType=Capsule, length=5.086, radius=1.642, center=<-0.0001 0 -1.6854>, rotation=<rot 0 -90 0>}]
 pack [
  {shapeType=Box, extentX=16.1049, extentY=7.5, extentZ=9.6207, center=<0 -20.2254 -19.1039>}
  {shapeType=Box, extentX=21.3313, extentY=12.665, extentZ=7.5, center=<1.4394 -9.7345 -21.0836>}
 ]
}
constraints [
 {bone1Name=dyn_gun, bone2Name=pack, swing1LimitDegrees=10, swing2LimitDegrees=10, twistLimitDegrees=5}
 {bone1Name=dyn_gernade, bone2Name=pack, swing1LimitDegrees=10, swing2LimitDegrees=10, twistLimitDegrees=10}
 {bone1Name=dyn_keychain_01, bone2Name=pack, swing1LimitDegrees=80, swing2LimitDegrees=80, twistLimitDegrees=5}
 {bone1Name=dyn_keychain_02, bone2Name=dyn_keychain_01, swing1LimitDegrees=80, swing2LimitDegrees=80, twistLimitDegrees=5}
 {bone1Name=dyn_pickaxe, bone2Name=pack, swing1LimitDegrees=10, swing2LimitDegrees=10, twistLimitDegrees=10}
]
