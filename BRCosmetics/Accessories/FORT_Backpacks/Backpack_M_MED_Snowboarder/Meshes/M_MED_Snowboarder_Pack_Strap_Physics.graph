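# Physics asset for the Snowboarder backpack cosmetic's bullet strap, setting up collision bodies and joints for physical motion of the strap.
# A box body on the pack bone anchors a chain of three thin box bodies on the dyn_bulletStrap bones.
# Each strap segment is constrained to the one before it, with moderate swing limits and very tight twist, so the strap can bend but barely twists.
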PhysicsAsset /BRCosmetics/Accessories/FORT_Backpacks/Backpack_M_MED_Snowboarder/Meshes/M_MED_Snowboarder_Pack_Strap_Physics
bodies {
 dyn_bulletStrap_01 [{shapeType=Box, extentX=5.9853, extentY=7.4242, extentZ=0.8629, center=<3.0291 0 0>}]
 dyn_bulletStrap_02 [{shapeType=Box, extentX=5.9853, extentY=7.4242, extentZ=0.8629, center=<3.0291 0 0>}]
 dyn_bulletStrap_03 [{shapeType=Box, extentX=5.9853, extentY=7.2246, extentZ=0.8629, center=<3.0191 0 0>}]
 pack [{shapeType=Box, extentX=16.1049, extentY=7.5, extentZ=9.6207, center=<0 -20.2254 -19.1039>}]
}
constraints [
 {bone1Name=dyn_bulletStrap_01, bone2Name=pack, swing1LimitDegrees=45, swing2LimitDegrees=45, twistLimitDegrees=3}
 {bone1Name=dyn_bulletStrap_02, bone2Name=dyn_bulletStrap_01, swing1LimitDegrees=45, swing2LimitDegrees=45, twistLimitDegrees=3}
 {bone1Name=dyn_bulletStrap_03, bone2Name=dyn_bulletStrap_02, swing1LimitDegrees=45, swing2LimitDegrees=45, twistLimitDegrees=3}
]
